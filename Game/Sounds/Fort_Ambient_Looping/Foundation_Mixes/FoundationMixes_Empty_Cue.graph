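# FoundationMixes_Empty_Cue is an empty SoundCue among the looping ambient Foundation Mixes sounds.
# It has no root and no nodes, so it references no waves and produces no audio.
# Its name suggests it serves as a silent stand-in where a foundation mix cue is expected.
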SoundCue /Game/Sounds/Fort_Ambient_Looping/Foundation_Mixes/FoundationMixes_Empty_Cue
root=null
nodes []
